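# Sound cue for the looping Carmine skydive sound, layering a close and a distant loop.
# A quality-level switch mixes the standard waves for the first two levels and "opt" variants for the other two.
# Each layer gets a short volume fade-in, random pitch, its own close or far attenuation asset and a third-person ranged-weapon-fire sound class.
# A concurrency override caps it at two instances, stopping the oldest.
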SoundCue /FNGameplayCues/Sounds/Fort_GamePlay_Sounds/Carmine/Carmine_SkydiveLoop_Cue
root=QualityLevel_0
nodes [
 {
  id=Attenuation_0
  children [Modulator_2]
  properties {attenuationSettings=/Game/Sounds/Attenuation/Fort_Weapon_Carmine_Close_Athena_Att, attenuationOverrides {bSpatialize=false}}
 }
 {
  id=Attenuation_1
  children [Modulator_0]
  properties {attenuationSettings=&Fort_Weapon_Carmine_Close_Athena_Att, attenuationOverrides {bSpatialize=false}}
 }
 {
  id=Attenuation_2
  children [Modulator_1]
  properties {attenuationSettings=/Game/Sounds/Attenuation/Fort_Weapon_Carmine_Far_Athena_Att, attenuationOverrides {bSpatialize=false}}
 }
 {
  id=Attenuation_3
  children [Modulator_3]
  properties {attenuationSettings=&Fort_Weapon_Carmine_Far_Athena_Att, attenuationOverrides {bSpatialize=false}}
 }
 {
  id=Enveloper_0
  children [WavePlayer_8]
  properties {pitchCurve {editorCurveData {keys [{time=0}]}}, volumeCurve {editorCurveData {keys [<at 0 0.1>,<at 0.4 1>]}}}
 }
 {
  id=Enveloper_1
  children [WavePlayer_0]
  properties {pitchCurve {editorCurveData {keys [{time=0}]}}, volumeCurve {editorCurveData {keys [<at 0 0.1>,<at 0.4 1>]}}}
 }
 {
  id=Enveloper_2
  children [WavePlayer_1]
  properties {pitchCurve {editorCurveData {keys [{time=0}]}}, volumeCurve {editorCurveData {keys [<at 0 0.1>,<at 0.4 1>]}}}
 }
 {
  id=Enveloper_3
  children [WavePlayer_2]
  properties {pitchCurve {editorCurveData {keys [{time=0}]}}, volumeCurve {editorCurveData {keys [<at 0 0.1>,<at 0.4 1>]}}}
 }
 {id=Mixer_0, children [SoundClass_0,SoundClass_1], properties {inputVolume [1,1]}}
 {id=Mixer_1, children [SoundClass_2,SoundClass_3], properties {inputVolume [1,1]}}
 {id=Modulator_0, children [Enveloper_0], properties {pitchMax=1.1, pitchMin=0.9}}
 {id=Modulator_1, children [Enveloper_1], properties {pitchMax=1.1, pitchMin=0.9}}
 {id=Modulator_2, children [Enveloper_2], properties {pitchMax=1.1, pitchMin=0.9}}
 {id=Modulator_3, children [Enveloper_3], properties {pitchMax=1.1, pitchMin=0.9}}
 {id=QualityLevel_0, children [Mixer_0,Mixer_0,Mixer_1,Mixer_1]}
 {
  id=SoundClass_0
  children [Attenuation_1]
  properties {soundClassOverride=&Fort_WeaponRanged_Fire_3P_PSM}
 }
 {
  id=SoundClass_1
  children [Attenuation_2]
  properties {soundClassOverride=&Fort_WeaponRanged_Fire_3P_PSM}
 }
 {
  id=SoundClass_2
  children [Attenuation_0]
  properties {soundClassOverride=&Fort_WeaponRanged_Fire_3P_PSM}
 }
 {
  id=SoundClass_3
  children [Attenuation_3]
  properties {soundClassOverride=&Fort_WeaponRanged_Fire_3P_PSM}
 }
 {id=WavePlayer_0, properties {bLooping=true, soundWaveAssetPtr=/Game/Sounds/Fort_GamePlay_Sounds/Carmine/Carmine_SkyDive_Loop_01_distant}}
 {
  id=WavePlayer_1
  properties {bLooping=true, soundWaveAssetPtr=&/Carmine_SkyDive_Loop_01_opt}
 }
 {
  id=WavePlayer_2
  properties {bLooping=true, soundWaveAssetPtr=&/Carmine_SkyDive_Loop_01_distant_opt}
 }
 {id=WavePlayer_8, properties {bLooping=true, soundWaveAssetPtr=/Game/Sounds/Fort_GamePlay_Sounds/Carmine/Carmine_SkyDive_Loop_01}}
]
properties {bOverrideConcurrency=true, maxDistance=25000, volumeMultiplier=0.7, concurrencyOverrides {maxCount=2, resolutionRule=StopOldest}}
